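# Niagara effect for the Flapjack Wrangler glider, with wingtip-style ribbon trails and dripping syrup.
# FN_AnimTrail draws two bone-pair ribbons, gated by full deployment and forward dot.
# SyrupDrops randomly bursts single sprite/mesh drops from six Bits sockets, sized per socket.
# In gameplay, drops cling to their local spawn point before blending into falling motion; front-end modifiers substitute gentler drift, and a Z-difference value drives camera offset and color.
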NiagaraSystem /BRCosmetics/Gadgets/Assets/VinderTech_GliderChute/Glider_FlapjackWranglerMale/FX/NS_Glider_Flapjack_Wrangler
dependencies {
 usedDataInterfaces [
  &NiagaraDataInterfaceCurve
  &NiagaraDataInterfaceArrayFloat
  &NiagaraDataInterfaceVectorCurve
  &NiagaraDataInterfaceCamera
  &NiagaraDataInterfaceColorCurve
 ]
 usedDynamicInputs []
 usedModules [&SystemState]
}
emitters {
 FN_AnimTrail {
  emitterSpawnScript [&FN_RibbonIDOffsetUtility()]
  emitterUpdateScript [
   &EmitterState()
   &FN_SpawnAnimTrail(
    Enable:NiagaraBool=&Boolean_AndOperation(
     Bool1=$User.bisFullyDeployed
     Bool2=&SetBoolByFloatComparison(A=$User.ForwardDot, B=0.2, Comparison Type=::NewEnumerator2)
    )
    Number of Anim Trails=2
   )
  ]
  particleSpawnScript [
   &FN_InitializeAnimTrailBoneIndecies(First Bone or Socket Index=1, Second Bone or Socket Index=0)
   &FN_InitializeAnimTrailBoneIndecies(AnimTrail Selection=1, First Bone or Socket Index=3, Second Bone or Socket Index=2) {module=FN_InitializeAnimTrailBoneIndecies001}
   &FN_InitializeAnimTrail(Color=<rgba 1 1 1 0.15>, Lifetime=0.333)
   &AddVelocity(Velocity=&Multiply_Vector(
     A=<0 400 0>
     B=&VectorFromFloat(Value=&MakeCustomFloatFromBool(Bool=$User.bIsFrontEndPreview, True Float=1, False Float=0))
    ))
  ]
  particleUpdateScript [
   &SolveForcesAndVelocity(Clamp Velocity=false, Limit Acceleration=false)
   &FN_UpdateAnimTrail(Locked Particle RibbonWidth=1, SkeletalMesh=$User.SkeletalMesh) {module=FN_UpdateAnimTrail001}
   /SetVariables(Particles.RibbonU0Override=$Particles.NormalizedAge) {isSetParameters=true}
   &ScaleColor(
    Scale Alpha=&FloatFromCurve(FloatCurve=DataInterface() {propertyValues {Curve {keys [<at 0 0 RCIM_Cubic>,<at 0.05 1 RCIM_Cubic>,<at 1 0 RCIM_Cubic>]}}})
    ScaleA=true
    ScaleRGB=true
   )
  ]
  rendererClasses [&NiagaraRibbonRendererProperties]
  renderers [{rendererClass=&NiagaraRibbonRendererProperties}]
 }
 SyrupDrops {
  emitterUpdateScript [
   &EmitterState(
    Life Cycle Mode=::NewEnumerator1
    Loop Duration=&RandomRangeFloat(Minimum=0.25, Maximum=2)
    Recalculate Duration Each Loop=true
    UseLoopDelay=false
   )
   &SpawnRate(SpawnRate=1000, Use Spawn Probability=false) {enabled=false}
   /SetVariables(Emitter.PlayerVelocityVector:Vector3f=$Engine.Owner.Velocity) {isSetParameters=true}
   /SetVariables(Emitter.PlayerSpeed:NiagaraFloat=&VectorLength(Vector=$Emitter.PlayerVelocityVector)) {isSetParameters=true}
   &SpawnPerUnit(
    Module.Use Max Movement Threshold=true
    Module.Use Movement Tolerance=false
    Spawn Per Unit:NiagaraFloat=&Lerp_Float(
     A=40
     B=100
     Alpha=&FloatFromCurve(FloatCurve=DataInterface() {propertyValues {Curve {keys [<at 0 1>,<at 1 0>]}}}, CurveIndex=$Emitter.PlayerSpeed)
    )
    Use Spawn Probability=false
   ) {enabled=false}
   &SpawnBurst_Instantaneous(Spawn Count=1, Spawn Probability=0.5, Use Spawn Probability=true)
  ]
  particleSpawnScript [
   /SetVariables(Particles.WhichSocket:NiagaraInt32=&RandomRangeInt(Minimum=0, Maximum=5)) {isSetParameters=true}
   /Niagara/Modules/Spawn/Initialization/InitializeParticle(
    Color=<rgba 1 1 1 2>
    Lifetime=&Multiply_Float(
     A=&FN_FrontEndFloatModifier(Gameplay=1, Front End=0.2)
     B=&RandomRangeFloat(Minimum=2, Maximum=3)
    )
    Mesh Scale=&Multiply_VectorByFloat(
     Vector=&Multiply_VectorByFloat(
      Vector=<1 1 1>
      Float=&Multiply_Float(
       A=4
       B=&SelectFloatFromArray(Direct Array Index=$Particles.WhichSocket, Float Selection Array=DataInterface() {propertyValues {FloatData [0.9,1,0.75,0.55,0.5,0.7]}})
      )
     )
     Float=&RandomRangeFloat(Minimum=1, Maximum=1.3)
    )
    Module.Write Color=true
    Module.Write Lifetime=true
    Module.Write Mass=true
    Module.Write Material Random=true
    Module.Write Position=false
    Module.Write Scale=true
    Module.Write SpriteRotation=false
    Module.Write SpriteSize=true
    Position=&SimulationPosition()
    Sprite Size=<5 5>
   )
   &SkeletalMeshLocation(
    Bone / Socket Index=$Particles.WhichSocket
    Inherit Velocity=false
    Interpolate Mesh Sampling=false
    MaskSpawnGroup=false
    Mesh Sampling Type=::NewEnumerator5
    Sampled Position Offset=<0 0 -1>
    Skeletal Mesh=$User.Bits
    Socket Sampling Mode=::NewEnumerator5
   )
   /SetVariables(Particles.LocalPosition:Vector3f=/Niagara/DynamicInputs/Transforms/TransformPosition(Position=Linked() {linkedVariable {name=Particles.Position, type=NiagaraPosition}}, Source Space=::World, Destination Space=::Local)) {isSetParameters=true}
   &ApplyInitialForces()
   &AddVelocity(Velocity=&FN_FrontEndVector3Modifier(Gameplay=<0 500 0>, Front End=<0 0 0>))
   /SetVariables(Particles.SocketFloat:NiagaraFloat=&MakeFloatFromInt(Int=$Particles.WhichSocket)) {isSetParameters=true}
  ]
  particleUpdateScript [
   /Niagara/Modules/Update/Orientation/V2/OrientMeshToVector(Override Quat=false, UseOptionalSideAxis=true, UseOptionalUpAxis=true)
   &InheritVelocity(Module.Limit Speed=true, Velocity Scale=&Multiply_VectorByFloat(Vector=<1 1 1>, Float=-0.1)) {module=InheritVelocity001}
   &ParticleState()
   &GravityForce(Gravity=&Multiply_VectorByFloat(
     Vector=<1 1 -700>
     Float=&FloatFromCurve(FloatCurve=DataInterface() {propertyValues {Curve {keys [{arriveTangent=1.8039, interpMode=RCIM_Cubic, leaveTangent=1.8039, tangentMode=RCTM_User, time=0, value=0}, <at 1 1>]}}})
    )) {enabled=false}
   &InheritVelocity(
    Module.Limit Speed=true
    Velocity Limit=150
    Velocity Scale=&Multiply_VectorByFloat(
     Vector=<1 1 1>
     Float=/Niagara/DynamicInputs/Math/LerpMultipleFloats(
      Float Count:ENiagara_MultipleLerpCount=::NewEnumerator3
      1:NiagaraFloat=0.02
      2:NiagaraFloat=0
      3:NiagaraFloat=?
      4:NiagaraFloat=?
      5:NiagaraFloat=?
      Alpha:NiagaraFloat=$Particles.NormalizedAge
     )
    )
   ) {enabled=false}
   &ScaleMeshSize(Scale Factor=&Multiply_Vector(
     A=<1 1 1>
     B=&VectorFromCurve(VectorCurve=DataInterface() {propertyValues {
        XCurve {keys [
          <at 0 1 RCIM_Cubic>
          {arriveTangent=0.5, interpMode=RCIM_Cubic, leaveTangent=-5.3087, tangentMode=RCTM_Break, time=0.2366, value=1.9311}
          {arriveTangent=11.709, interpMode=RCIM_Cubic, leaveTangent=11.709, tangentMode=RCTM_User, time=0.3386, value=1.5078}
          <at 1 0.5>
         ]}
        YCurve {keys [
          {arriveTangent=2.3356, interpMode=RCIM_Cubic, time=-0.0034, value=1}
          {arriveTangent=5.9046, interpMode=RCIM_Cubic, leaveTangent=-6.5667, tangentMode=RCTM_Break, time=0.2412, value=1.7614}
          {arriveTangent=-1.48, interpMode=RCIM_Cubic, leaveTangent=-1.48, tangentMode=RCTM_User, time=0.3329, value=1.3319}
          <at 1 1>
         ]}
        ZCurve {keys [
          {arriveTangent=1.8377, interpMode=RCIM_Cubic, leaveTangent=1.8377, tangentMode=RCTM_User, time=-0.0042, value=1}
          {arriveTangent=5.4767, interpMode=RCIM_Cubic, leaveTangent=-3.8259, tangentMode=RCTM_Break, time=0.244, value=1.7314}
          {arriveTangent=-1.4717, interpMode=RCIM_Cubic, leaveTangent=-1.4717, tangentMode=RCTM_User, time=0.3368, value=1.3453}
          <at 1 1>
         ]}
       }})
    ))
   &AccelerationForce(Acceleration=&FN_FrontEndVector3Modifier(Gameplay=<0 0 -980>, Front End=<0 0 0>))
   &AddVelocity(Velocity=&FN_FrontEndVector3Modifier(
     Gameplay=<0 0 0>
     Front End=&Multiply_Vector(
      A=<0.3 -50 25>
      B=&VectorFromCurve(VectorCurve=DataInterface() {propertyValues {
         XCurve {keys [<at 0 1>,<at 1 1>]}
         YCurve {keys [<at 0 1>,<at 1 1>]}
         ZCurve {keys [<at 0 0.25 RCIM_Cubic>, {arriveTangent=3.5645, interpMode=RCIM_Cubic, leaveTangent=3.5645, tangentMode=RCTM_User, time=0.65, value=2}]}
        }})
     )
    )) {module=AddVelocity001}
   &SolveForcesAndVelocity(Clamp Velocity=false, Limit Acceleration=false)
   /SetVariables(Particles.Position=&FN_FrontEndVector3Modifier(
     Gameplay=&Lerp_Vector(
      A=&TransformPosition(Position=$Particles.LocalPosition, Source Space=::Local, Destination Space=::Simulation)
      B=Linked() {linkedVariable {name=Particles.Position, type=NiagaraPosition}}
      Alpha=&FloatFromCurve(FloatCurve=DataInterface() {propertyValues {Curve {keys [
           {interpMode=RCIM_Cubic, tangentMode=RCTM_Break, time=0.25, value=0}
           {arriveTangent=3.8697, interpMode=RCIM_Cubic, tangentMode=RCTM_Break, time=0.5, value=1}
          ]}}})
     )
     Front End=Linked() {linkedVariable {name=Particles.Position, type=NiagaraPosition}}
    )) {isSetParameters=true}
   &ScaleColor(
    Scale Alpha=&FloatFromCurve(FloatCurve=DataInterface() {propertyValues {Curve {keys [<at 0 1 RCIM_Cubic>,<at 1 0>]}}})
    ScaleA=true
    ScaleRGB=true
   ) {enabled=false}
   &DynamicMaterialParameters(
    Index 0 Param 1=0.8
    Index 0 Param 2=&FloatFromCurve(FloatCurve=DataInterface() {propertyValues {Curve {keys [{arriveTangent=-1.8884, interpMode=RCIM_Cubic, leaveTangent=-1.8884, tangentMode=RCTM_User, time=0, value=1}, <at 1 0.26>]}}})
    Index 0 Param 3=1
    Index 0 Param 4=1
    Module.Param0WriteEnabled=true
    Module.Param1WriteEnabled=true
    Module.Param2WriteEnabled=true
    Module.Param3WriteEnabled=true
   )
   &self:CameraOffsetAngle() {module=CameraOffsetAngle}
   &CameraOffset(Camera Offset Amount=&Multiply_Float(A=3, B=&ClampFloat(Float=$Particles.ZDiffValue, Min=-1, Max=5)))
   &GenerateLocationEvent(
    Event Send Rate=1
    EventCondition:NiagaraBool=&SetBoolByFloatComparison(A=$Particles.NormalizedAge, B=0)
    Float to Send as Random Float:NiagaraFloat=$Particles.SocketFloat
    Module.Use Event Send Rate=true
   )
   &Color(Color=&ColorFromCurve(
     ColorCurve=DataInterface() {propertyValues {
       AlphaCurve {keys [<at 0 1 RCIM_Cubic>]}
       BlueCurve {keys [<at -1 0>,<at -0.4 0>,<at 0.4 0.0546>,<at 1 1 RCIM_Cubic>]}
       GreenCurve {keys [<at -1 0>,<at -0.4 0.9106>,<at 0.4 1>,<at 1 0.0111 RCIM_Cubic>]}
       RedCurve {keys [<at -1 1>,<at -0.4 1>,<at 0.4 0>,<at 1 0 RCIM_Cubic>]}
      }}
     CurveIndex=&Multiply_Float(A=$Particles.ZDiffValue, B=0.01)
    ))
  ]
  rendererClasses [&NiagaraSpriteRendererProperties, &NiagaraMeshRendererProperties]
  renderers [
   {rendererClass=&NiagaraSpriteRendererProperties}
   {rendererClass=&NiagaraMeshRendererProperties, rendererIndex=1}
  ]
 }
}
summary {
 systemName=NS_Glider_Flapjack_Wrangler
 emitters [
  {emitterName=SyrupDrops, rendererClasses [&NiagaraSpriteRendererProperties, &NiagaraMeshRendererProperties]}
  {emitterName=FN_AnimTrail, rendererClasses [&NiagaraRibbonRendererProperties]}
 ]
 userVariables [
  User.bIsFrontEnd {type=NiagaraBool}
  User.bIsFrontEndPreview {type=NiagaraBool}
  User.bisFullyDeployed {type=NiagaraBool}
  User.Bits {
   type=NiagaraDataInterfaceSkeletalMesh
   defaultValue {
    struct=NiagaraExt_VariableValue_DataInterface
    value {dataInterface=&self:NiagaraDataInterfaceSkeletalMesh_0}
   }
  }
  User.ForwardDot {type=NiagaraFloat, defaultValue=0, description=""}
  User.JumpSpawn {type=NiagaraFloat, defaultValue=0, description=""}
  User.RibbonIDOffset {type=NiagaraInt32}
  User.SkeletalMesh {
   type=NiagaraDataInterfaceSkeletalMesh
   defaultValue {
    struct=NiagaraExt_VariableValue_DataInterface
    value {dataInterface=&self:NiagaraDataInterfaceSkeletalMesh_1}
   }
  }
 ]
}
