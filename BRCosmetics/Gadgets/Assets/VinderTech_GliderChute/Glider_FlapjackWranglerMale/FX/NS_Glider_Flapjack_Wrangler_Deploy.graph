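# Niagara deploy effect for the Flapjack Wrangler glider, spawning short-lived sprite puffs from the skeletal mesh passed in as User.Bits.
# Two near-identical emitters fire a small burst plus a decaying spawn rate, differing in rate and a raised sampling offset on one.
# Particles push outward from a point, then drag and gravity act while curves grow, fade and drive dynamic material parameters.
# Tint brightness follows the NPC time-of-day parameter.
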NiagaraSystem /BRCosmetics/Gadgets/Assets/VinderTech_GliderChute/Glider_FlapjackWranglerMale/FX/NS_Glider_Flapjack_Wrangler_Deploy
dependencies {
 usedDataInterfaces [&NiagaraDataInterfaceCurve, &NiagaraDataInterfaceColorCurve, &NiagaraDataInterfaceVector2DCurve]
 usedDynamicInputs []
 usedModules [&SystemState]
}
emitters {
 Empty002 {
  emitterUpdateScript [
   &EmitterState()
   &SpawnRate(
    SpawnRate=&Multiply_Float(
     A=30
     B=&FloatFromCurve(
      FloatCurve=DataInterface() {propertyValues {Curve {keys [{arriveTangent=-2.5855, interpMode=RCIM_Cubic, leaveTangent=-2.5855, tangentMode=RCTM_User, time=0, value=1}, <at 1 0>]}}}
      CurveIndex=$Emitter.NormalizedLoopAge
     )
    )
    Use Spawn Probability=false
   )
   &SpawnBurst_Instantaneous(Spawn Count=&RandomRangeInt(Minimum=2, Maximum=3), Use Spawn Probability=false)
  ]
  particleSpawnScript [
   /Niagara/Modules/Spawn/Initialization/InitializeParticle(
    Color=&MultiplyLinearColor(
     Linear Color=<rgba 0.5 0.3615 0.1406 3>
     Scale Factor=&Vector4FromCurve(
      CurveIndex=Linked() {linkedVariable {name=NPC.FortniteNPC.TimeOfDay_0-1, type=NiagaraFloat}}
      Vector4Curve=DataInterface() {propertyValues {
        AlphaCurve {keys [<at 0 1>]}
        BlueCurve {keys [<at 0 0.5>,<at 0.4796 1>,<at 1 0.5>]}
        GreenCurve {keys [<at 0 0.5>,<at 0.4796 1>,<at 1 0.5>]}
        RedCurve {keys [<at 0 0.5>,<at 0.4796 1>,<at 1 0.5>]}
       }}
     )
    )
    Lifetime=&RandomRangeFloat(Minimum=0.2, Maximum=0.38)
    Module.Write Color=true
    Module.Write Lifetime=true
    Module.Write Mass=true
    Module.Write Material Random=true
    Module.Write Position=true
    Module.Write Scale=false
    Module.Write SpriteRotation=false
    Module.Write SpriteSize=true
    Sprite Size=&Multiply_Vector2(A=<0.5 1>, B=&Vector2DFromFloat(Value=450))
   )
   /SetVariables(Particles.LocalPosition:Vector3f=/Niagara/DynamicInputs/Transforms/TransformPosition(Position=Linked() {linkedVariable {name=Particles.Position, type=NiagaraPosition}}, Source Space=::World, Destination Space=::Local)) {enabled=false, isSetParameters=true, module=SetVariables_#1}
   &SubUVAnimation(End Frame=3, SubUV Animation Mode:ENiagara_SubUVLookupMode=::NewEnumerator2)
   &SphereLocation(
    MaskSpawnGroup=false
    Offset Coordinate Space:ENiagaraCoordinateSpace=::Simulation
    Override Seed=false
    Sphere Radius=500
    Surface Only=false
    UseNonUniformScale=false
    UseOffset=true
    UseSphereOrientation=false
   ) {enabled=false}
   &AddVelocityFromPoint(
    Module.Use Falloff Distance=false
    Velocity Strength=&Multiply_Float(A=350, B=&RandomRangeFloat(Minimum=0.35, Maximum=1))
   )
   &SkeletalMeshLocation(
    Bone Sampling Mode=::NewEnumerator0
    Inherit Velocity=false
    MaskSpawnGroup=false
    Module.Override Randomness=false
    Skeletal Mesh=$User.Bits
   )
  ]
  particleUpdateScript [
   &Drag(Use Linear Drag=true, Use Rotational Drag=false)
   &GravityForce()
   &SolveForcesAndVelocity(Clamp Velocity=false, Limit Acceleration=false)
   &DynamicMaterialParameters(
    Index 0 Param 1=&FloatFromCurve(FloatCurve=DataInterface() {propertyValues {Curve {keys [
         {arriveTangent=-0.8342, interpMode=RCIM_Cubic, leaveTangent=-0.8342, tangentMode=RCTM_User, time=0, value=1}
         {arriveTangent=-0.5745, interpMode=RCIM_Cubic, leaveTangent=-0.5745, tangentMode=RCTM_User, time=1, value=0}
        ]}}})
    Index 0 Param 2=&FloatFromCurve(FloatCurve=DataInterface() {propertyValues {Curve {keys [<at 0 0.05 RCIM_Cubic>, {arriveTangent=0.4316, interpMode=RCIM_Cubic, leaveTangent=0.4316, tangentMode=RCTM_User, time=1, value=0.2}]}}})
    Index 0 Param 3=1
    Index 0 Param 4=1
    Module.Param0WriteEnabled=true
    Module.Param1WriteEnabled=true
    Module.Param2WriteEnabled=true
    Module.Param3WriteEnabled=true
   )
   &ParticleState()
   &ScaleSpriteSize(Scale Factor=&Vector2DFromCurve(
     Vector2Curve=DataInterface() {propertyValues {
       XCurve {keys [<at 0 0.2 RCIM_Cubic>, {arriveTangent=2.1844, interpMode=RCIM_Cubic, leaveTangent=2.1844, tangentMode=RCTM_User, time=1, value=1.5}]}
       YCurve {keys [<at 0 0.2>,<at 1 1>]}
      }}
     ScaleCurve=<1 1>
    ))
   &ScaleColor(
    Scale Alpha=&FloatFromCurve(FloatCurve=DataInterface() {propertyValues {Curve {keys [<at 0 1 RCIM_Cubic>,<at 1 0>]}}})
    ScaleA=true
    ScaleRGB=true
   )
   /SetVariables(Particles.Position=&Lerp_Vector(
     A=&TransformPosition(Position=$Particles.LocalPosition, Source Space=::Local, Destination Space=::Simulation)
     B=Linked() {linkedVariable {name=Particles.Position, type=NiagaraPosition}}
     Alpha=&FloatFromCurve(FloatCurve=DataInterface() {propertyValues {Curve {keys [<at 0 0 RCIM_Cubic>, {arriveTangent=4.4737, interpMode=RCIM_Cubic, tangentMode=RCTM_Break, time=0.3, value=1}]}}})
    )) {enabled=false, isSetParameters=true, module=SetVariables_#2}
  ]
  rendererClasses [&NiagaraSpriteRendererProperties]
  renderers [
   {rendererClass=&NiagaraSpriteRendererProperties}
   {rendererClass=&NiagaraSpriteRendererProperties, rendererIndex=1}
  ]
 }
 Empty003 {
  emitterUpdateScript [
   &EmitterState()
   &SpawnRate(
    SpawnRate=&Multiply_Float(
     A=50
     B=&FloatFromCurve(
      FloatCurve=DataInterface() {propertyValues {Curve {keys [{arriveTangent=-2.5855, interpMode=RCIM_Cubic, leaveTangent=-2.5855, tangentMode=RCTM_User, time=0, value=1}, <at 1 0>]}}}
      CurveIndex=$Emitter.NormalizedLoopAge
     )
    )
    Use Spawn Probability=false
   )
   &SpawnBurst_Instantaneous(Spawn Count=&RandomRangeInt(Minimum=2, Maximum=3), Use Spawn Probability=false)
  ]
  particleSpawnScript [
   &InitializeParticle(
    Color=&MultiplyLinearColor(
     Linear Color=<rgba 0.5 0.3615 0.1406 3>
     Scale Factor=&Vector4FromCurve(
      CurveIndex=Linked() {linkedVariable {name=NPC.FortniteNPC.TimeOfDay_0-1, type=NiagaraFloat}}
      Vector4Curve=DataInterface() {propertyValues {
        AlphaCurve {keys [<at 0 1>]}
        BlueCurve {keys [<at 0 0.5>,<at 0.4796 1>,<at 1 0.5>]}
        GreenCurve {keys [<at 0 0.5>,<at 0.4796 1>,<at 1 0.5>]}
        RedCurve {keys [<at 0 0.5>,<at 0.4796 1>,<at 1 0.5>]}
       }}
     )
    )
    Lifetime=&RandomRangeFloat(Minimum=0.2, Maximum=0.38)
    Module.Write Color=true
    Module.Write Lifetime=true
    Module.Write Mass=true
    Module.Write Material Random=true
    Module.Write Position=true
    Module.Write Scale=false
    Module.Write SpriteRotation=false
    Module.Write SpriteSize=true
    Sprite Size=&Multiply_Vector2(A=<0.5 1>, B=&Vector2DFromFloat(Value=450))
   )
   /SetVariables(Particles.LocalPosition:Vector3f=&TransformPosition(Position=Linked() {linkedVariable {name=Particles.Position, type=NiagaraPosition}}, Source Space=::World, Destination Space=::Local)) {enabled=false, isSetParameters=true, module=SetVariables_#1}
   &SubUVAnimation(End Frame=3, SubUV Animation Mode:ENiagara_SubUVLookupMode=::NewEnumerator2)
   &SphereLocation(
    MaskSpawnGroup=false
    Offset Coordinate Space:ENiagaraCoordinateSpace=::Simulation
    Override Seed=false
    Sphere Radius=500
    Surface Only=false
    UseNonUniformScale=false
    UseOffset=true
    UseSphereOrientation=false
   ) {enabled=false}
   &AddVelocityFromPoint(
    Module.Use Falloff Distance=false
    Velocity Strength=&Multiply_Float(A=350, B=&RandomRangeFloat(Minimum=0.35, Maximum=1))
   )
   &SkeletalMeshLocation(
    Bone Sampling Mode=::NewEnumerator0
    Inherit Velocity=false
    MaskSpawnGroup=false
    Module.Override Randomness=false
    Sampled Position Offset=<0 0 70>
    Skeletal Mesh=$User.Bits
   )
  ]
  particleUpdateScript [
   &Drag(Use Linear Drag=true, Use Rotational Drag=false)
   &GravityForce()
   &SolveForcesAndVelocity(Clamp Velocity=false, Limit Acceleration=false)
   &DynamicMaterialParameters(
    Index 0 Param 1=&FloatFromCurve(FloatCurve=DataInterface() {propertyValues {Curve {keys [
         {arriveTangent=-0.8342, interpMode=RCIM_Cubic, leaveTangent=-0.8342, tangentMode=RCTM_User, time=0, value=1}
         {arriveTangent=-0.5745, interpMode=RCIM_Cubic, leaveTangent=-0.5745, tangentMode=RCTM_User, time=1, value=0}
        ]}}})
    Index 0 Param 2=&FloatFromCurve(FloatCurve=DataInterface() {propertyValues {Curve {keys [<at 0 0.05 RCIM_Cubic>, {arriveTangent=0.4316, interpMode=RCIM_Cubic, leaveTangent=0.4316, tangentMode=RCTM_User, time=1, value=0.2}]}}})
    Index 0 Param 3=1
    Index 0 Param 4=1
    Module.Param0WriteEnabled=true
    Module.Param1WriteEnabled=true
    Module.Param2WriteEnabled=true
    Module.Param3WriteEnabled=true
   )
   &ParticleState()
   &ScaleSpriteSize(Scale Factor=&Vector2DFromCurve(
     Vector2Curve=DataInterface() {propertyValues {
       XCurve {keys [<at 0 0.2 RCIM_Cubic>, {arriveTangent=2.1844, interpMode=RCIM_Cubic, leaveTangent=2.1844, tangentMode=RCTM_User, time=1, value=1.5}]}
       YCurve {keys [<at 0 0.2>,<at 1 1>]}
      }}
     ScaleCurve=<1 1>
    ))
   &ScaleColor(
    Scale Alpha=&FloatFromCurve(FloatCurve=DataInterface() {propertyValues {Curve {keys [<at 0 1 RCIM_Cubic>,<at 1 0>]}}})
    ScaleA=true
    ScaleRGB=true
   )
   /SetVariables(Particles.Position=&Lerp_Vector(
     A=&TransformPosition(Position=$Particles.LocalPosition, Source Space=::Local, Destination Space=::Simulation)
     B=Linked() {linkedVariable {name=Particles.Position, type=NiagaraPosition}}
     Alpha=&FloatFromCurve(FloatCurve=DataInterface() {propertyValues {Curve {keys [<at 0 0 RCIM_Cubic>, {arriveTangent=4.4737, interpMode=RCIM_Cubic, tangentMode=RCTM_Break, time=0.3, value=1}]}}})
    )) {enabled=false, isSetParameters=true, module=SetVariables_#2}
  ]
  rendererClasses [&NiagaraSpriteRendererProperties]
  renderers [
   {rendererClass=&NiagaraSpriteRendererProperties}
   {rendererClass=&NiagaraSpriteRendererProperties, rendererIndex=1}
  ]
 }
}
summary {
 systemName=NS_Glider_Flapjack_Wrangler_Deploy
 emitters [
  {emitterName=Empty002, rendererClasses [&NiagaraSpriteRendererProperties]}
  {emitterName=Empty003, rendererClasses [&NiagaraSpriteRendererProperties]}
 ]
 userVariables [User.Bits {
   type=NiagaraDataInterfaceSkeletalMesh
   defaultValue {
    struct=NiagaraExt_VariableValue_DataInterface
    value {dataInterface=&self:NiagaraDataInterfaceSkeletalMesh_0}
   }
  }]
}
